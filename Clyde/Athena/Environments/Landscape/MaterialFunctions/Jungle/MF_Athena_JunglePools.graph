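# Material function for the Athena jungle landscape that recolors a base color with the JunglePools grass tint.
# It fully desaturates the BaseColor input, then multiplies it by the RGB of JunglePools_GrassTint from the MPC_Landscape parameter collection.
# The Mask input blends between the original and tinted color, defaulting to fully tinted.
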MaterialFunction /Clyde/Athena/Environments/Landscape/MaterialFunctions/Jungle/MF_Athena_JunglePools
graph {nodes [
  n1 = FunctionInput() {inputName=BaseColor}
  n2 = Desaturation(None=@n1, Fraction=Constant() {r=1})
  n3 = ComponentMask(None=CollectionParameter() {collection=/Game/Athena/Environments/Landscape/MPC/MPC_Landscape, parameterName=JunglePools_GrassTint}) {b=true, g=true, r=true}
  n4 = Constant() {r=1}
  FunctionOutput(None=LinearInterpolate(
    A=@n1
    B=Multiply(A=@n2, B=@n3)
    Alpha=FunctionInput(Preview=@n4) {bUsePreviewValueAsDefault=true, inputName=Mask, inputType=FunctionInput_Scalar, sortPriority=1}
   ))
 ]}
